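# Material function that offsets UVs into a horizontal flipbook strip by camera direction around a particle, so a flat texture can show view-dependent frames as a fake 3D look.
# The normalized camera-minus-particle vector's XY becomes an angle via VectorToRadialValue, scaled by rotations per world rotation and image count, then floored to a frame step.
# That step offsets U, which is compressed to one image's width; V passes through unchanged.
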MaterialFunction /Engine/Functions/Engine_MaterialFunctions02/Texturing/3DParticleUVs
graph {
 properties {
  description=Moves through a flipbook texture based on the cameras position compared to the object center.
  libraryCategoriesText [Misc,Texturing]
 }
 nodes [
  n1 = Subtract(A=CameraPositionWS(), B=ParticlePositionWS())
  n2 = MaterialFunctionCall(Vector or UVs=ComponentMask(None=Normalize(VectorInput=@n1)) {g=true, r=true}) {materialFunction=&VectorToRadialValue}
  n3 = FunctionInput() {
   bUsePreviewValueAsDefault=true
   inputName=Number of texture rotations per world rotation
   inputType=FunctionInput_Scalar
   previewValue=<1 1 0 1>
  }
  n4 = FunctionInput() {inputName=Number of horizontal Images, inputType=FunctionInput_Scalar, previewValue=<1 0 0 1>}
  n5 = Multiply(A=@n3, B=@n4)
  n6 = Divide(A=Floor(None=Multiply(A=@n2.Vector Converted to Angle, B=@n5)), B=@n4)
  n7 = FunctionInput(Preview=TextureCoordinate()) {
   bUsePreviewValueAsDefault=true
   description="Default Value = UV coordinate 0"
   inputName=UVs
   inputType=FunctionInput_Vector2
   previewValue=<1 0 0 1>
  }
  n8 = Divide(A=ComponentMask(None=@n7) {r=true}, B=@n4)
  FunctionOutput(None=AppendVector(A=Add(A=@n6, B=@n8), B=ComponentMask(None=@n7) {g=true})) {outputName=Uvs}
 ]
}
