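# Example material demonstrating the engine's FlowMaps function, animating water diffuse and normal textures along a swirl flow map.
# Red and green of the swirl texture are recentered as flow vectors, while scrolling blurred noise offsets the time input to hide repetition.
# Distortion alpha scales the blend from flat to flowing normal.
# A feature level switch substitutes a flat normal on ES2 and ES3.1.
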Material /Engine/Functions/Engine_MaterialFunctions02/ExampleContent/FlowMaps
graph {
 properties {pixelDepthOffsetMode=PDOM_Legacy, usedShadingModels=Default Lit}
 nodes [
  Constant()
  n1 = Subtract(A=ComponentMask(None=TextureSample() {samplerType=SAMPLERTYPE_LinearColor, texture=/Engine/Functions/Engine_MaterialFunctions02/ExampleContent/Textures/SwirlMap}.RGB) {g=true, r=true}) {constB=0.5}
  n2 = TextureCoordinate() {uTiling=2, vTiling=2}
  n3 = MaterialFunctionCall(Speed=Constant() {r=0.5}) {materialFunction=&TimeWithSpeedVariable}
  n4 = Multiply(A=@n3.Time, B=Constant() {r=0.02})
  n5 = TextureSample!(UVs=Frac(None=Add(A=@n2, B=@n4)), Tex, Level, Apply View MipBias) {
   desc=noise to break up the repetition
   mipValueMode=TMVM_MipLevel
   samplerType=SAMPLERTYPE_LinearGrayscale
   texture=/Engine/Functions/Engine_MaterialFunctions02/ExampleContent/Textures/LowResBlurredNoise
  }
  n6 = MaterialFunctionCall!(
   Diffuse=TextureObject() {texture=/Engine/Functions/Engine_MaterialFunctions02/ExampleContent/Textures/water_d}
   Normal Map=TextureObject() {samplerType=SAMPLERTYPE_Normal, texture=/Engine/Functions/Engine_MaterialFunctions02/ExampleContent/Textures/water_n}
   "Flow Vector Map (see tooltip)"=@n1
   UV texture 2 offset
   "Time (see tooltip)"=Add(A=Multiply(A=@n5.R) {constB=0.25}, B=@n3.Time)
   useMipLevel
   UVs
   Texture Size
   Texture Mip Bias
  ) {materialFunction=/Engine/Functions/Engine_MaterialFunctions02/Texturing/Flowmaps.FlowMaps}
  n7 = Multiply(A=ComponentMask(None=@n6.Distortion) {a=true}) {constB=50}
  n8 = LinearInterpolate(A=Constant3Vector() {constant=<rgba 0 0 1 0>}, B=@n6.Normal, Alpha=Clamp(None=Add(A=@n7) {constB=0.1}))
  n9 = Constant3Vector() {constant=<rgba 0 0 1 0>}
  FeatureLevelSwitch(Default=@n8, ES2=@n9, ES3_1=@n9, SM4_REMOVED=@n8)
  n10 = TextureSample!(UVs=TextureCoordinate(), Tex, Apply View MipBias) {
   desc=noise to break up the repeatition
   samplerType=SAMPLERTYPE_LinearGrayscale
   texture=&LowResBlurredNoise
  }
  Multiply(A=@n1, B=@n10.R)
 ]
}
