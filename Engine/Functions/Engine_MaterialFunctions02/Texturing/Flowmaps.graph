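# FlowMaps material function that pushes diffuse and normal textures along flow map vectors, with blended UV and distortion outputs.
# Two fractional time phases half a cycle apart scale the flow vector into UV offsets, and LinearSine crossfades their samples to hide resets.
# Default time adds low-res noise to break up tiling.
# A static switch selects computed, biased mip levels for sampling; leftover normal-reorientation nodes feed no output.
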MaterialFunction /Engine/Functions/Engine_MaterialFunctions02/Texturing/Flowmaps.FlowMaps
graph {
 properties {
  bExposeToLibrary=true
  description=Pushes the Diffuse and Normal textures along the flow map vectors. See Material'/Engine/Functions/Engine_MaterialFunctions02/ExampleContent/FlowMaps' for setting suggestions or additional techniques.
  libraryCategoriesText [Misc,Texturing]
 }
 nodes [
  n1 = TextureCoordinate()
  n2 = MaterialFunctionCall(Speed=Constant() {r=0.5}) {materialFunction=&TimeWithSpeedVariable}
  n3 = Multiply(A=@n2.Time, B=Constant() {r=0.02})
  n4 = TextureSample!(UVs=Frac(None=Add(A=@n1, B=@n3)), Tex, Level, Apply View MipBias) {
   desc=noise to break up the repetition
   mipValueMode=TMVM_MipLevel
   samplerType=SAMPLERTYPE_LinearGrayscale
   texture=/Engine/Functions/Engine_MaterialFunctions02/ExampleContent/Textures/LowResBlurredNoise
  }
  n5 = FunctionInput(Preview=Add(A=Multiply(A=@n4.R) {constB=0.25}, B=@n2.Time)) {
   bUsePreviewValueAsDefault=true
   description=Adding a large scale noise map to time will help offset any tiling artifacts.
   inputName="Time (see tooltip)"
   inputType=FunctionInput_Scalar
   sortPriority=9
   previewValue=<0.2 0 0 1>
  }
  n6 = FunctionInput() {
   bUsePreviewValueAsDefault=true
   description=Flow direction from -1 to 1. Modulating the flowmap with a macro noise texture may help reduce tiling artifacts. Please note: There is a discrepancy between 3DS Max/Mayas representation of uvs and Unreals. The green channel is inverted. If the provided Photoshop action is used to generate the flow vector maps then this inversion will be corrected. Also note that any textures used as flow maps should have sRGB in the texture properties window disabled. Using low resolution textures with vector displacement texture compression enabled is preferable.
   inputName="Flow Vector Map (see tooltip)"
   inputType=FunctionInput_Vector2
   previewValue=<0.1 0 0 1>
  }
  n7 = Multiply(A=@n6, B=Constant2Vector() {g=1, r=-1})
  n8 = Multiply(A=Frac(None=Subtract(A=@n5) {constB=0.5, desc=phase offset}), B=@n7)
  n9 = MaterialFunctionCall(f2=Abs(None=@n8)) {materialFunction=&AddComponents}
  n10 = Multiply(A=@n9.f2, B=@n9.f2) {desc=a controllable power node is expensive and would add complication to the functions interface}
  n11 = Frac(None=@n5)
  n12 = Multiply(A=@n7, B=@n11)
  n13 = MaterialFunctionCall(f2=Abs(None=@n12)) {materialFunction=&AddComponents}
  n14 = Multiply(A=@n13.f2, B=@n13.f2) {desc=a controllable power node is expensive and would add complication to the functions interface}
  n15 = AppendVector(A=@n10, B=@n14)
  n16 = MaterialFunctionCall(Value=@n11, Sine Phase=StaticBool()) {desc=Transition, materialFunction=&LinearSine}
  FunctionOutput(None=AppendVector(A=AppendVector(A=@n15, B=@n16.Linear Sine), B=LinearInterpolate(A=@n10, B=@n14, Alpha=@n16.Linear Sine))) {
   description="R= amount of distortion for lerp sample A. G= distortion amount for Lerp input B.  B= Lerp Alpha. A= Lerped Distortion amounts. This is useful for modulating normals by the amount of movement that occurs in the texture. I.e. Lerp ((0,0,1), (This nodes normal output), (This nodes distortions alpha channel)) to only display the normals where the waves are currently highest."
   outputName=Distortion
   sortPriority=3
  }
  n17 = FunctionInput(Preview=TextureCoordinate() {uTiling=5, vTiling=5}) {bUsePreviewValueAsDefault=true, inputName=UVs, inputType=FunctionInput_Vector2, sortPriority=10}
  n18 = Add(A=@n17, B=@n8)
  n19 = FunctionInput() {bUsePreviewValueAsDefault=true, inputName=UV texture 2 offset, inputType=FunctionInput_Vector2, sortPriority=2, previewValue=<0.5 0.5 0 1>}
  n20 = Add(A=Frac(None=@n18), B=@n19) {constB=0.5}
  n21 = FunctionInput(Preview=TextureObject() {texture=/Engine/EngineMaterials/DefaultDiffuse}) {
   bUsePreviewValueAsDefault=true
   description="This can be any texture not just a diffuse map. The rgb and alpha values are returned. "
   inputName=Diffuse
   inputType=FunctionInput_Texture2D
   sortPriority=-2
  }
  n22 = FunctionInput() {
   bUsePreviewValueAsDefault=true
   description="Input the diffuse or normal maps XY Size. (1024x1024 for example)"
   inputName=Texture Size
   inputType=FunctionInput_Vector2
   sortPriority=11
   previewValue=<1024 1024 0 1>
  }
  n23 = FunctionInput() {
   bUsePreviewValueAsDefault=true
   description=Numbers larger than one will make your diffuse and normal texture maps textures blurry. This is useful for shorelines.
   inputName=Texture Mip Bias
   inputType=FunctionInput_Scalar
   sortPriority=12
  }
  n24 = Add(
   A=MaterialFunctionCall!(UVs=@n17, Texture Size=@n22) {materialFunction=/Engine/Functions/Engine_MaterialFunctions01/Texturing/ComputeMipLevel}.MipLevel
   B=@n23
  )
  n25 = TextureSample(UVs=@n20, Tex=@n21, Level=@n24) {mipValueMode=TMVM_MipLevel}
  n26 = TextureSample!(UVs=@n20, Tex=@n21, Apply View MipBias)
  n27 = FunctionInput(Preview=StaticBool() {value=true}) {
   bUsePreviewValueAsDefault=true
   desc=Allows you to disable to use of MIP level override so this can be used in vertex shader
   inputName=useMipLevel
   inputType=FunctionInput_StaticBool
   sortPriority=10
  }
  n28 = Add(A=@n12, B=@n17)
  n29 = TextureSample!(UVs=@n28, Tex=@n21, Level=@n24, Apply View MipBias) {mipValueMode=TMVM_MipLevel}
  n30 = TextureSample!(UVs=@n28, Tex=@n21, Apply View MipBias)
  n31 = LinearInterpolate(
   A=StaticSwitch(True=AppendVector(A=@n25.RGB, B=@n25.RGB), False=AppendVector(A=@n26.RGB, B=@n26.RGB), Value=@n27)
   B=StaticSwitch(True=AppendVector(A=@n29.RGB, B=@n29.RGB), False=AppendVector(A=@n30.RGB, B=@n30.RGB), Value=@n27)
   Alpha=@n16.Linear Sine
  )
  FunctionOutput(None=ComponentMask(None=@n31) {a=true}) {outputName=Diffuse Alpha, sortPriority=1}
  FunctionOutput(None=LinearInterpolate(A=@n18, B=@n28, Alpha=@n16.Linear Sine)) {
   description="R= amount of distortion for lerp sample A. G= distortion amount for Lerp input B.  B= Lerp Alpha. A= Lerped Distortion amounts. This is useful for modulating normals by the amount of movement that occurs in the texture. I.e. Lerp ((0,0,1), (This nodes normal output), (This nodes distortions alpha channel)) to only display the normals where the waves are currently highest."
   outputName=UVs
   sortPriority=8
  }
  n32 = FunctionInput(Preview=TextureObject() {samplerType=SAMPLERTYPE_Normal, texture=/Engine/EngineMaterials/DefaultNormal}) {bUsePreviewValueAsDefault=true, inputName=Normal Map, inputType=FunctionInput_Texture2D, sortPriority=-1}
  n33 = LinearInterpolate(
   A=StaticSwitch(
    True=TextureSample!(UVs=@n20, Tex=@n32, Level=@n24, Apply View MipBias) {mipValueMode=TMVM_MipLevel}.RGB
    False=TextureSample!(UVs=@n20, Tex=@n32, Apply View MipBias).RGB
    Value=@n27
   )
   B=StaticSwitch(
    True=TextureSample!(UVs=@n28, Tex=@n32, Level=@n24, Apply View MipBias) {mipValueMode=TMVM_MipLevel}.RGB
    False=TextureSample!(UVs=@n28, Tex=@n32, Apply View MipBias).RGB
    Value=@n27
   )
   Alpha=@n16.Linear Sine
  ) {desc=normal}
  FunctionOutput(None=@n33) {outputName=Normal, sortPriority=2}
  FunctionOutput(None=ComponentMask(None=@n31) {b=true, g=true, r=true}) {outputName=Diffuse}
  n34 = Multiply(A=AppendVector(A=LinearInterpolate(), B=Constant()), B=Constant() {r=1})
  n35 = MaterialFunctionCall!(
   Vector1=Add(A=@n34, B=Constant3Vector() {constant=<rgba 1 0 0 0>})
   Vector2=Add(A=@n34, B=Constant3Vector() {constant=<rgba 0 1 0 0>})
   Vector3=Constant3Vector() {constant=<rgba 0 0 1 0>}
  ) {materialFunction=/Engine/Functions/Engine_MaterialFunctions02/Math/MakeVectorsOrthogonal}
  MaterialFunctionCall(BasisX=@n35.Vector1, BasisY=@n35.Vector1, BasisZ=@n35.Vector1) {materialFunction=&Transform3x3Matrix}
  n36 = Normalize(VectorInput=AppendVector(A=@n7, B=Constant() {r=1}))
  Normalize(VectorInput=Add(
    A=MaterialFunctionCall(BaseNormal=@n36, AdditionalNormal=@n33) {materialFunction=&BlendAngleCorrectedNormals}.Result
    B=Constant3Vector() {constant=<rgba 0 0 1 0>}
   ))
 ]
}
