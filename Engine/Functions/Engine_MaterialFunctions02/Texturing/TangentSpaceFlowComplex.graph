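# Material function that animates a FlowTexture by pushing its UVs along a tangent-space FlowMap direction.
# Two samples scroll on sawtooth time cycles offset by half a phase, the second also shifted half a tile, and cross-fade through LinearSine to hide resets.
# Blurred noise scales flow strength and, by default, offsets animation time to break up repetition.
# Several leftover nodes do not feed the output.
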MaterialFunction /Engine/Functions/Engine_MaterialFunctions02/Texturing/TangentSpaceFlowComplex
graph {
 properties {description=Pushing the FlowTexture along the direction of the FlowDirection, libraryCategoriesText [Misc,Texturing]}
 nodes [
  n1 = TextureSample(UVs=TextureCoordinate()) {
   desc=noise to break up the repetition
   samplerType=SAMPLERTYPE_LinearGrayscale
   texture=/Engine/Functions/Engine_MaterialFunctions02/ExampleContent/Textures/LowResBlurredNoise
  }
  n2 = FunctionInput(Preview=@n1.R) {bUsePreviewValueAsDefault=true, inputName=Anim offset Noise, inputType=FunctionInput_Scalar, sortPriority=10, previewValue=<0.2 0 0 1>}
  n3 = FunctionInput(Preview=MaterialFunctionCall(Speed=Constant() {r=-0.25}) {materialFunction=&TimeWithSpeedVariable}.Time) {bUsePreviewValueAsDefault=true, inputName=Time, inputType=FunctionInput_Scalar, sortPriority=10, previewValue=<0.2 0 0 1>}
  n4 = Add(A=@n2, B=@n3)
  n5 = Frac(None=@n4)
  n6 = MaterialFunctionCall(Value=@n5, Sine Phase=StaticBool()) {materialFunction=&LinearSine}
  n7 = TextureSample() {samplerType=SAMPLERTYPE_LinearGrayscale, texture=&LowResBlurredNoise}
  n8 = ComponentMask(None=@n7.RGB) {g=true, r=true}
  n9 = FunctionInput(Preview=Multiply(A=Subtract(A=@n8) {constB=0.5}) {constB=3}) {bUsePreviewValueAsDefault=true, description=Flow Direction in Tangent Space from -1 to 1, inputName=FlowMap, inputType=FunctionInput_Vector2}
  n10 = TextureSample(UVs=TextureCoordinate()) {
   desc=noise to break up the repeatition
   samplerType=SAMPLERTYPE_LinearGrayscale
   texture=&LowResBlurredNoise
  }
  n11 = Multiply(A=@n9, B=@n10.R)
  n12 = FunctionInput(Preview=TextureCoordinate() {uTiling=6, vTiling=6}) {bUsePreviewValueAsDefault=true, inputName=UVs, inputType=FunctionInput_Vector2}
  n13 = FunctionInput(Preview=TextureObject() {texture=/Engine/EngineMaterials/DefaultDiffuse}) {inputName=FlowTexture, inputType=FunctionInput_Texture2D}
  n14 = TextureSample!(UVs=Add(A=Multiply(A=@n11, B=@n5), B=@n12), Tex=@n13, Level, Apply View MipBias) {mipValueMode=TMVM_MipLevel}
  n15 = Subtract(A=@n4) {constB=0.5, desc=phase offset}
  n16 = Add(A=@n12, B=Multiply(A=Frac(None=@n15), B=@n11))
  n17 = TextureSample(UVs=Add(A=Frac(None=@n16)) {constB=0.5}, Tex=@n13) {mipValueMode=TMVM_MipLevel}
  Add(A=Multiply(A=@n6.Linear Sine, B=@n14.RGB), B=Multiply(A=OneMinus(None=@n6.Linear Sine), B=@n17.RGB))
  Add(A=@n11, B=@n11)
  FunctionOutput(None=LinearInterpolate(A=@n17.RGB, B=@n14.RGB, Alpha=@n6.Linear Sine))
  LinearInterpolate(Alpha=@n10.R) {constA=0.75, constB=1.25}
  TextureObject() {texture=null}
  TextureSample!(UVs=TextureCoordinate(), Tex, Apply View MipBias) {
   desc=noise to break up the repeatition
   samplerType=SAMPLERTYPE_LinearGrayscale
   texture=&LowResBlurredNoise
  }
  TextureSample(UVs=TextureCoordinate()) {
   desc=noise to break up the repeatition
   samplerType=SAMPLERTYPE_LinearGrayscale
   texture=&LowResBlurredNoise
  }
 ]
}
