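# Material function that blends two colors or textures using two grayscale height maps and a transition value, exposed in the Misc and Texturing library categories.
# It offsets the difference between the heights by the transition phase and clamps it, so the second input emerges as the phase rises.
# CheapContrast sharpens this alpha, which lerps both the diffuse inputs and the height maps.
# Outputs are the blended result, alpha, pre-contrast alpha and combined height.
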MaterialFunction /Engine/Functions/Engine_MaterialFunctions02/Texturing/HeightLerpWithTwoHeightMaps
graph {
 properties {
  bExposeToLibrary=true
  description=Lerp between two values or textures based off of two height map and a transition value.
  libraryCategoriesText [Misc,Texturing]
 }
 nodes [
  n1 = FunctionInput() {
   bUsePreviewValueAsDefault=true
   description=Enter a grayscale heightmap with values between 0-1.
   inputName=Height Texture 1
   inputType=FunctionInput_Scalar
   sortPriority=3
  }
  n2 = FunctionInput() {
   bUsePreviewValueAsDefault=true
   description=Enter a value between 0-1.
   inputName=Transistion Phase
   inputType=FunctionInput_Scalar
   sortPriority=2
  }
  n3 = Add(A=@n1, B=OneMinus(None=@n2))
  n4 = Multiply(B=@n2) {constA=2}
  n5 = FunctionInput() {
   bUsePreviewValueAsDefault=true
   description=Enter a grayscale heightmap with values between 0-1.
   inputName=Height Texture 2
   inputType=FunctionInput_Scalar
   sortPriority=4
  }
  n6 = Clamp(None=Subtract(A=Add(A=@n3), B=Add(A=@n4, B=@n5)))
  FunctionOutput(None=ComponentMask(None=@n6) {r=true}) {description=Value used in lerp function. Positive 1 lerps to texture 1., outputName=Lerp Alpha No Contrast, sortPriority=2}
  n7 = FunctionInput() {
   bUsePreviewValueAsDefault=true
   description=Uses cheap contrast function and adds one instruction.
   inputName=Contrast
   inputType=FunctionInput_Scalar
   sortPriority=5
  }
  n8 = ComponentMask(None=MaterialFunctionCall(In=@n6, Contrast=@n7) {materialFunction=&CheapContrast}.Result) {r=true}
  FunctionOutput(None=LinearInterpolate(A=@n5, B=@n1, Alpha=@n8) {desc=lerp 0 texture 2}) {description="Value used in lerp function. ", outputName=Resulting Height Map, sortPriority=3}
  n9 = FunctionInput() {
   bUsePreviewValueAsDefault=true
   description=Enter a color or texture. This texture is exposed as the Transistion Phase increases.
   inputName=Diffuse 2
   sortPriority=1
  }
  n10 = FunctionInput() {
   bUsePreviewValueAsDefault=true
   description=Enter a color or texture. This texture Acts as the base.
   inputName=Diffuse 1
   previewValue=<1 1 1 1>
  }
  FunctionOutput(None=LinearInterpolate(A=@n9, B=@n10, Alpha=@n8) {desc=lerp 0 texture 2}) {description=Lerped result of value a and b based off of the input heightmap and transition value., outputName=Results}
  FunctionOutput(None=@n8) {description=Value used in lerp function. Positive 1 lerps to texture 1, outputName=Alpha, sortPriority=1}
 ]
}
