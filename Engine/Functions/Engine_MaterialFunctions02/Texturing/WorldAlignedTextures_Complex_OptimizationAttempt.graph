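# Material function for world-aligned triplanar tiling of diffuse and normal textures, with combined, XY-only and Z-only outputs.
# World position over texture size gives per-axis UVs, blended by contrast-adjusted world normal, optionally refined by height lerp, with optional separate Z size and diffuse texture.
# Switches select WorldAlignedNormals_HighQuality or inline normals, in world or tangent space.
# The vertex-shader optimization switch and its Custom UV outputs are unconnected.
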MaterialFunction /Engine/Functions/Engine_MaterialFunctions02/Texturing/WorldAlignedTextures_Complex_OptimizationAttempt
graph {
 properties {description=Tiles a texture in worldspace, libraryCategoriesText [Misc,Texturing]}
 nodes [
  FunctionInput() {bUsePreviewValueAsDefault=true, description=-------------, inputName=-------------, inputType=FunctionInput_StaticBool, sortPriority=-10}
  FunctionInput() {
   bUsePreviewValueAsDefault=true
   description=-------------
   inputName="     -------------"
   inputType=FunctionInput_StaticBool
   sortPriority=199
  }
  FunctionInput() {bUsePreviewValueAsDefault=true, description=-------------, inputName="  -------------", inputType=FunctionInput_StaticBool, sortPriority=104}
  FunctionInput() {
   bUsePreviewValueAsDefault=true
   description=-------------
   inputName="      -------------"
   inputType=FunctionInput_StaticBool
   sortPriority=5
  }
  FunctionInput() {
   bUsePreviewValueAsDefault=true
   description=-------------
   inputName="    -------------"
   inputType=FunctionInput_StaticBool
   sortPriority=10
  }
  FunctionInput(Preview=StaticBool()) {
   bUsePreviewValueAsDefault=true
   description="This switch defaults to false. This switch moves several operations out of the pixel shader and into the vertex shader. Which should be faster to run in most cases.  Enable high quality normals and connect all of the UV outputs. from this function to the corresponding custom uv inputs on the base shader. Multiple world aligned normal functions can use this technique at the same time. Only one node needs to connect to the uv inputs.       "
   inputName="OptimizeHighQualityNormals (see comment)"
   inputType=FunctionInput_StaticBool
   sortPriority=31
  }
  n1 = FunctionInput(Preview=WorldPosition().XYZ) {bUsePreviewValueAsDefault=true, inputName=WorldPosition, sortPriority=200}
  n2 = FunctionInput() {
   bUsePreviewValueAsDefault=true
   description=Size of texture in worldspace units
   inputName=TextureSize
   sortPriority=-11
   previewValue=<64 64 64 6>
  }
  n3 = Multiply(A=Abs(None=@n2)) {constB=-1}
  n4 = Divide(A=@n1, B=@n3)
  n5 = ComponentMask(None=@n4) {b=true, r=true}
  n6 = TextureObject() {texture=/Engine/EngineMaterials/DefaultDiffuse}
  n7 = FunctionInput(Preview=@n6) {bUsePreviewValueAsDefault=true, inputName=Diffuse Texture Object, inputType=FunctionInput_Texture2D, sortPriority=-100}
  n8 = TextureSample(UVs=@n5, Tex=@n7) {desc=Y projection}
  n9 = AppendVector(A=@n8.RGB, B=@n8.RGB)
  n10 = FunctionInput(Preview=StaticBool()) {
   bUsePreviewValueAsDefault=true
   description="Default = False"
   inputName=Export Float 4
   inputType=FunctionInput_StaticBool
   sortPriority=105
  }
  n11 = ComponentMask(None=@n4) {b=true, g=true}
  n12 = TextureSample(UVs=@n11, Tex=@n7) {desc=X projection}
  n13 = AppendVector(A=@n12.RGB, B=@n12.RGB)
  n14 = FunctionInput(Preview=VertexNormalWS()) {bUsePreviewValueAsDefault=true, inputName=World Space Normal, sortPriority=201, previewValue=<0 0 1 1>}
  n15 = ComponentMask(None=@n14) {r=true}
  n16 = FunctionInput() {bUsePreviewValueAsDefault=true, inputName=ProjectionTransitionContrast, inputType=FunctionInput_Scalar, sortPriority=7}
  n17 = ComponentMask(None=MaterialFunctionCall(In=Abs(None=@n15), Contrast=@n16) {materialFunction=&CheapContrast}.Result) {desc="1=+x", r=true}
  n18 = FunctionInput() {
   bUsePreviewValueAsDefault=true
   description="(If pick the diffuse height channel == true)  (1,0,0,0) will use the diffuse red channel as the height map. (0,1,0,0) will use the green channel etc."
   inputName=TheDiffuseHeightChannel
   inputType=FunctionInput_Vector4
   sortPriority=9
  }
  n19 = FunctionInput(Preview=StaticBool()) {
   bUsePreviewValueAsDefault=true
   description="Default = False"
   inputName=PickTheDiffuseHeightChannel
   inputType=FunctionInput_StaticBool
   sortPriority=9
  }
  n20 = FunctionInput() {bUsePreviewValueAsDefault=true, inputName=HeightLerpTransitionContrast, inputType=FunctionInput_Scalar, sortPriority=9}
  n21 = MaterialFunctionCall!(
   Diffuse 1
   Diffuse 2
   Transistion Phase=@n17
   Height Texture 1=StaticSwitch(True=DotProduct(A=@n9, B=@n18), False=@n8.A, Value=@n19)
   Height Texture 2=StaticSwitch(True=DotProduct(A=@n13, B=@n18), False=@n12.A, Value=@n19)
   Contrast=@n20
  ) {materialFunction=&/HeightLerpWithTwoHeightMaps}
  n22 = OneMinus(None=@n21.Alpha)
  n23 = FunctionInput(Preview=StaticBool()) {
   bUsePreviewValueAsDefault=true
   description=Defaults to false.
   inputName=Use Texture Alpha Heightmap to Refine the Transistion
   inputType=FunctionInput_StaticBool
   sortPriority=8
   previewValue=<1 0 0 1>
  }
  n24 = LinearInterpolate(
   A=StaticSwitch(True=@n9, False=@n8.RGB, Value=@n10)
   B=StaticSwitch(True=@n13, False=@n12.RGB, Value=@n10)
   Alpha=StaticSwitch(True=@n22, False=@n17, Value=@n23)
  ) {desc="Float3 1 = x projection"}
  n25 = FunctionInput() {
   bUsePreviewValueAsDefault=true
   description=Size of texture in worldspace units
   inputName=Texture Size Z projection
   previewValue=<64 64 64 6>
  }
  n26 = Multiply(A=@n25) {constB=-1}
  n27 = FunctionInput(Preview=StaticBool()) {
   bUsePreviewValueAsDefault=true
   inputName=Use seperate settings for Z projection
   inputType=FunctionInput_StaticBool
   previewValue=<64 64 64 6>
  }
  n28 = ComponentMask(None=StaticSwitch(True=Divide(A=@n1, B=@n26), False=@n4, Value=@n27)) {g=true, r=true}
  n29 = FunctionInput(Preview=@n6) {bUsePreviewValueAsDefault=true, inputName=Z Plane Diffuse Texture Object, inputType=FunctionInput_Texture2D, sortPriority=4}
  n30 = TextureSample(UVs=@n28, Tex=@n29)
  n31 = TextureSample(UVs=@n28, Tex=@n7)
  n32 = FunctionInput(Preview=StaticBool()) {
   bUsePreviewValueAsDefault=true
   inputName=Use seperate texture for Z projection
   inputType=FunctionInput_StaticBool
   previewValue=<64 64 64 6>
  }
  n33 = StaticSwitch(True=AppendVector(A=@n30.RGB, B=@n30.RGB), False=AppendVector(A=@n31.RGB, B=@n31.RGB), Value=@n32) {desc=Z projection}
  n34 = StaticSwitch(True=@n33, False=ComponentMask(None=@n33) {b=true, g=true, r=true}, Value=@n10)
  n35 = ComponentMask(None=@n14) {b=true}
  n36 = ComponentMask(None=MaterialFunctionCall(In=Abs(None=@n35), Contrast=@n16) {materialFunction=&CheapContrast}.Result) {desc="1=+z", r=true}
  n37 = DotProduct(A=@n33, B=@n18)
  n38 = ComponentMask(None=@n33) {a=true}
  n39 = OneMinus(None=MaterialFunctionCall(
    Transistion Phase=@n36
    Height Texture 1=@n21.Resulting Height Map
    Height Texture 2=StaticSwitch(True=@n37, False=@n38, Value=@n19)
    Contrast=@n20
   ) {materialFunction=&HeightLerpWithTwoHeightMaps}.Alpha)
  FunctionOutput(None=LinearInterpolate(A=@n24, B=@n34, Alpha=StaticSwitch(True=@n39, False=@n36, Value=@n23)) {desc="1 = z projection"}) {outputName=XYZ Diffuse Texture, sortPriority=1}
  FunctionOutput() {outputName=-------, sortPriority=20}
  n40 = FunctionInput(Preview=TextureObject() {samplerType=SAMPLERTYPE_LinearColor, texture=/Engine/EngineMaterials/DefaultNormal_Uncompressed}) {bUsePreviewValueAsDefault=true, inputName=NormalTextureObject, inputType=FunctionInput_Texture2D, sortPriority=30}
  n41 = StaticSwitch(True=@n39, False=@n36, Value=@n23)
  n42 = StaticSwitch(True=@n22, False=@n17, Value=@n23)
  n43 = MaterialFunctionCall!(Negative World Position Divisor=@n3, World Position=@n1, Texture=@n40, Z lerp Alpha=@n41, WorldNormal=@n14, X lerp alpha=@n42) {materialFunction=/Engine/Functions/Engine_MaterialFunctions02/WorldAlignedNormals_HighQuality}
  n44 = Transform(None=@n43.World Space Normals XYZ) {transformSourceType=TRANSFORMSOURCE_World, transformType=TRANSFORM_Tangent}
  n45 = FunctionInput(Preview=StaticBool() {value=true}) {
   bUsePreviewValueAsDefault=true
   description="This switch defaults to true. If left true ensure that tangent space normals is turned off in the shaders properties panel.   . "
   inputName=World Space Normals
   inputType=FunctionInput_StaticBool
   sortPriority=31
  }
  n46 = TextureSample(UVs=@n5, Tex=@n40) {samplerType=SAMPLERTYPE_Normal}
  n47 = Multiply(A=@n46.R) {constB=-1}
  n48 = Multiply(A=@n14, B=Constant() {r=1})
  n49 = Multiply(A=@n46.B, B=ComponentMask(None=@n48) {g=true})
  n50 = TextureSample(UVs=@n11, Tex=@n40) {samplerType=SAMPLERTYPE_Normal}
  n51 = Multiply(A=ComponentMask(None=@n48) {r=true}, B=@n50.B)
  n52 = Multiply(A=@n50.R) {constB=-1}
  n53 = LinearInterpolate(
   A=AppendVector(A=AppendVector(A=@n47, B=@n49), B=Multiply(A=@n46.G) {constB=-1})
   B=AppendVector(A=AppendVector(A=@n51, B=@n52), B=Multiply(A=@n50.G) {constB=-1})
   Alpha=@n42
  ) {desc=side projections}
  n54 = TextureSample(UVs=@n28, Tex=@n40) {samplerType=SAMPLERTYPE_Normal}
  n55 = Constant3Vector() {constant=<rgba -1 -1 1 0>}
  n56 = AppendVector(A=ComponentMask(None=Multiply(A=@n54.RGB, B=@n55)) {g=true, r=true}, B=Multiply(A=@n54.B, B=ComponentMask(None=@n48) {b=true}))
  n57 = Normalize(VectorInput=LinearInterpolate(A=@n53, B=@n56, Alpha=@n41))
  n58 = Transform(None=@n57) {transformSourceType=TRANSFORMSOURCE_World, transformType=TRANSFORM_Tangent}
  n59 = FunctionInput(Preview=StaticBool() {value=true}) {
   bUsePreviewValueAsDefault=true
   description=This switch defaults to true
   inputName=UseHighQualityNormals
   inputType=FunctionInput_StaticBool
   sortPriority=31
  }
  FunctionOutput(None=StaticSwitch(
    True=StaticSwitch(True=@n43.World Space Normals XYZ, False=@n44, Value=@n45)
    False=StaticSwitch(True=@n57, False=@n58, Value=@n45)
    Value=@n59
   )) {outputName=Normal XYZ Texture, sortPriority=10}
  FunctionOutput(None=@n24) {outputName=XY Diffuse Texture, sortPriority=2}
  FunctionOutput(None=@n34) {outputName=Z Diffuse Texture, sortPriority=3}
  n60 = Transform(None=@n43.Z Projection World Space Normals) {transformSourceType=TRANSFORMSOURCE_World, transformType=TRANSFORM_Tangent}
  n61 = Normalize(VectorInput=@n56)
  n62 = Transform(None=@n61) {transformSourceType=TRANSFORMSOURCE_World, transformType=TRANSFORM_Tangent}
  FunctionOutput(None=StaticSwitch(
    True=StaticSwitch(True=@n43.Z Projection World Space Normals, False=@n60, Value=@n45)
    False=StaticSwitch(True=@n61, False=@n62, Value=@n45)
    Value=@n59
   )) {outputName=Normal Z Texture, sortPriority=13}
  FunctionOutput() {outputName=-------------, sortPriority=4}
  FunctionOutput() {outputName=Opti Normal - Custom UV 6, sortPriority=25}
  FunctionOutput() {outputName=Opti Normal - Custom UV 7, sortPriority=26}
  FunctionOutput() {outputName=Opti Normal - Custom UV 5, sortPriority=24}
  FunctionOutput() {outputName=Opti Normal - Custom UV 4, sortPriority=23}
  n63 = Transform(None=@n43.XY Projection World Space Normals) {transformSourceType=TRANSFORMSOURCE_World, transformType=TRANSFORM_Tangent}
  n64 = Normalize(VectorInput=@n53)
  n65 = Transform(None=@n64) {transformSourceType=TRANSFORMSOURCE_World, transformType=TRANSFORM_Tangent}
  FunctionOutput(None=StaticSwitch(
    True=StaticSwitch(True=@n43.XY Projection World Space Normals, False=@n63, Value=@n45)
    False=StaticSwitch(True=@n64, False=@n65, Value=@n45)
    Value=@n59
   )) {outputName=Normal XY Texture, sortPriority=12}
  FunctionOutput() {outputName=Opti Normal - Custom UV 3, sortPriority=22}
  FunctionOutput() {outputName=Opti Normal - Custom UV 2, sortPriority=21}
  n66 = Transform(None=@n43.XY Flat Top Projection World Space Normals) {transformSourceType=TRANSFORMSOURCE_World, transformType=TRANSFORM_Tangent}
  n67 = Normalize(VectorInput=LinearInterpolate(A=@n53, B=@n14, Alpha=@n41))
  n68 = Transform(None=@n67) {transformSourceType=TRANSFORMSOURCE_World, transformType=TRANSFORM_Tangent}
  FunctionOutput(None=StaticSwitch(
    True=StaticSwitch(True=@n43.XY Flat Top Projection World Space Normals, False=@n66, Value=@n45)
    False=StaticSwitch(True=@n67, False=@n68, Value=@n45)
    Value=@n59
   )) {outputName=Normal XYZ Flat Top, sortPriority=11}
 ]
}
