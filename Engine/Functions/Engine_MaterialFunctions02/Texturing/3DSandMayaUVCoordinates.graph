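# Material function that converts mesh UVs to the bottom-left origin used by 3ds Max and Maya, for cases where specific values are stored in UV channels.
# Each of the first four texture coordinate channels keeps U, inverts V with OneMinus, and is re-appended into its own output.
# Outputs UVChannel0 to UVChannel3 are zero-based, and their descriptions map them to Max's one-based channel numbers.
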MaterialFunction /Engine/Functions/Engine_MaterialFunctions02/Texturing/3DSandMayaUVCoordinates
graph {
 properties {
  bExposeToLibrary=true
  description="Inverts the green channel in each set of model uvs to place 0,0 in the bottom left hand corner to match 3D Studio Max and Mayas UV layouts. Flipping the values are important if you're storing specific values in the uv channels."
  libraryCategoriesText [Misc,Texturing]
 }
 nodes [
  n1 = TextureCoordinate()
  n2 = ComponentMask(None=@n1) {g=true}
  FunctionOutput(None=AppendVector(A=ComponentMask(None=@n1) {r=true}, B=OneMinus(None=@n2))) {description="Unreal channel 0, Max channel 1", outputName=UVChannel0}
  n3 = TextureCoordinate() {coordinateIndex=1}
  n4 = ComponentMask(None=@n3) {g=true}
  FunctionOutput(None=AppendVector(A=ComponentMask(None=@n3) {r=true}, B=OneMinus(None=@n4))) {description="Unreal channel 1, Max channel 2", outputName=UVChannel1, sortPriority=1}
  n5 = TextureCoordinate() {coordinateIndex=2}
  n6 = ComponentMask(None=@n5) {g=true}
  FunctionOutput(None=AppendVector(A=ComponentMask(None=@n5) {r=true}, B=OneMinus(None=@n6))) {description="Unreal channel 2, Max channel 3", outputName=UVChannel2, sortPriority=2}
  n7 = TextureCoordinate() {coordinateIndex=3}
  n8 = ComponentMask(None=@n7) {g=true}
  FunctionOutput(None=AppendVector(A=ComponentMask(None=@n7) {r=true}, B=OneMinus(None=@n8))) {description="Unreal channel 3, Max channel 4", outputName=UVChannel3, sortPriority=3}
 ]
}
