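# Material function that gives particle sprites depth-textured opacity, so they intersect scene geometry as if they had volume.
# A depth texture, scaled by particle radius times a ratio or by an explicit world-space depth, offsets pixel depth for a soft fade against scene depth.
# Static switches add particle alpha and near-camera falloff, optionally depth-aware.
# A second output gives a World Position Offset toward the camera, defaulting to the texture depth, to emulate camera offset.
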MaterialFunction /Engine/Functions/Engine_MaterialFunctions02/Texturing/3dParticleOpacity
graph {
 properties {bExposeToLibrary=true, libraryCategoriesText [Misc,Particles]}
 nodes [
  FunctionInput() {bUsePreviewValueAsDefault=true, inputName=-------WPO Group-------, inputType=FunctionInput_StaticBool, sortPriority=50}
  FunctionInput() {
   bUsePreviewValueAsDefault=true
   inputName=-------Camera Falloff Group-------
   inputType=FunctionInput_StaticBool
   sortPriority=10
   previewValue=<-500 0 0 1>
  }
  FunctionInput() {
   bUsePreviewValueAsDefault=true
   inputName=-------Particle Alpha Group-------
   inputType=FunctionInput_StaticBool
   sortPriority=30
   previewValue=<-500 0 0 1>
  }
  FunctionInput() {bUsePreviewValueAsDefault=true, inputName=-------------Depth Texture Options-----------, inputType=FunctionInput_StaticBool, sortPriority=1}
  n1 = PixelDepth()
  n2 = FunctionInput() {
   bUsePreviewValueAsDefault=true
   description="Enter the ratio of the height textures depth to the height textures width (use 3ds max world unit measurements). For instance, if you create a plane in a 3D application that is 512x512 and you render a height texture from that plane that represents 256 units away from the plane then you would calculate this number by dividing 256/512 which would be a ratio of .5. "
   inputName=Dynamic Texture Depth Ratio
   inputType=FunctionInput_Scalar
   sortPriority=3
   previewValue=<0.5 0 0 1>
  }
  n3 = FunctionInput() {
   bUsePreviewValueAsDefault=true
   description=World space scale of the depth texture. A value of 256 will make the depth texture 256 units deep.
   inputName="Explicit Texture Depth (WS)"
   inputType=FunctionInput_Scalar
   sortPriority=4
   previewValue=<128 0 0 1>
  }
  n4 = FunctionInput(Preview=StaticBool() {value=true}) {
   bUsePreviewValueAsDefault=true
   description="The default setting is dynamic texture depth. Using explicit texture settings will allow you to specifiy exactly how many world units your depth texture represents. For example 512 units would make the texture 512 units deep. The dynamic texture depth option will use the ratio between the particles size and it's depth to scale the depth for differently scaled particles. For example if your ratio was set to 1, a square particle that was 512 units would set the 512 unit depth texture. A ratio value of .5 would generate a texture depth of 256. This helps the effect scale properly with various particle sizes."
   inputName="Use dynamic (True) or explicit texture depth settings (False)"
   inputType=FunctionInput_StaticBool
   sortPriority=2
  }
  n5 = StaticSwitch(True=Multiply(A=ParticleRadius(), B=@n2), False=@n3, Value=@n4)
  n6 = TextureSample() {samplerType=SAMPLERTYPE_LinearColor, texture=/Engine/Functions/Engine_MaterialFunctions02/ExampleContent/Textures/newhplaneHeightMap}
  n7 = FunctionInput(Preview=@n6.R) {
   bUsePreviewValueAsDefault=true
   description="Add an inverted z-depth render (black near the camera and white values in the background). Be sure that the textures SRGB switch is disabled in the texture properties window."
   inputName=Depth Texture
   inputType=FunctionInput_Scalar
  }
  n8 = Multiply(A=@n7, B=@n5)
  n9 = Subtract(A=@n5, B=@n8)
  n10 = FunctionInput(Preview=StaticBool()) {
   bUsePreviewValueAsDefault=true
   description="Makes closer pixels (according to the depth map) fade out before farther pixels. The default value is false."
   inputName=Use Depth In Camera Falloff Calculation
   inputType=FunctionInput_StaticBool
   sortPriority=13
  }
  n11 = FunctionInput() {
   bUsePreviewValueAsDefault=true
   description=The distance in world space units from the camera where the object remains transparent.
   inputName=Near Camera Falloff Start Distance
   inputType=FunctionInput_Scalar
   sortPriority=11
   previewValue=<64 0 0 1>
  }
  n12 = Subtract(A=StaticSwitch(True=Subtract(A=@n1.R, B=@n9), False=@n1.R, Value=@n10), B=@n11)
  n13 = FunctionInput() {
   bUsePreviewValueAsDefault=true
   description=The distance from the camera where the opacity fades to 0. Enter the reciprocal of the target distance. For example to fade the particle out over 256 units enter 1/256 which Unreal will turn into 0.003906.
   inputName="Near Camera Fade Distance (1/n)"
   inputType=FunctionInput_Scalar
   sortPriority=12
   previewValue=<0.0078 0.0039 0 1>
  }
  n14 = SceneDepth()
  n15 = Add(A=@n1.R, B=@n8)
  n16 = FunctionInput() {
   bUsePreviewValueAsDefault=true
   description="The distance in world space units that the objects opacity gradually fades out across. Enter 1 divided by the number of world space units you would like the transition to take place across. For instance enter 1/16 in the input field and Unreal will convert that to 0.062500. Doing this ahead of time removes an instruction from the effect. "
   inputName="Depth Texture Falloff Softness (WS) (1/n)"
   inputType=FunctionInput_Scalar
   sortPriority=5
   previewValue=<0.0625 0.0625 0 1>
  }
  n17 = Clamp(None=Multiply(A=Subtract(A=@n14.R, B=@n15), B=@n16))
  n18 = FunctionInput(Preview=StaticBool()) {
   bUsePreviewValueAsDefault=true
   description=Multiplies the result by the particle color alpha channel. The default value is false.
   inputName=Use Particle Alpha
   inputType=FunctionInput_StaticBool
   sortPriority=31
  }
  n19 = StaticSwitch(True=Multiply(A=@n17, B=ParticleColor().A) {desc=*alpha}, False=@n17, Value=@n18)
  n20 = Multiply(A=Clamp(None=Multiply(A=@n12, B=@n13)) {desc=NearCameraFalloff}, B=@n19)
  n21 = FunctionInput(Preview=StaticBool()) {
   bUsePreviewValueAsDefault=true
   description=Makes the object falloff as they approach the camera. The default value is false.
   inputName=Use Near Camera Falloff
   inputType=FunctionInput_StaticBool
   sortPriority=11
  }
  FunctionOutput(None=Clamp(None=StaticSwitch(True=@n20, False=@n19, Value=@n21))) {outputName=Opacity}
  n22 = Subtract(A=CameraPositionWS(), B=ParticlePositionWS())
  n23 = FunctionInput(Preview=@n5) {
   bUsePreviewValueAsDefault=true
   description="This value should be the same value as \"Texture Depth.\" The \"Texture Depth\" value is used as this inputs default value. "
   inputName=Camera Offset
   inputType=FunctionInput_Scalar
   sortPriority=51
  }
  FunctionOutput(None=Multiply(A=Normalize(VectorInput=@n22), B=@n23)) {description="Emulates what camera offset does in the particle editor. ", outputName=World Position Offset - Camera Offset, sortPriority=1}
 ]
}
