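# Material function, exposed to the library, that rotates 2D values such as UVs or world coordinates around a chosen center, with angle 1 equal to one full turn.
# It subtracts the center, applies a 2D rotation matrix built from two dot products against sine and cosine vectors, then adds the center back.
# Sine and Cosine keep their default period, so the angle input needs no radian conversion.
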MaterialFunction /Engine/Functions/Engine_MaterialFunctions02/Texturing/CustomRotator
graph {
 properties {
  bExposeToLibrary=true
  description=Rotator with rotation center exposed and a rotation angle on 0-1 scale. A rotation angle value of 1 is equal to one full rotation.
  libraryCategoriesText [Misc,Texturing]
 }
 nodes [
  n1 = FunctionInput() {
   bUsePreviewValueAsDefault=true
   description=Enter a number to be used as the center point of rotation.
   inputName=Rotation Center
   inputType=FunctionInput_Vector2
   sortPriority=1
   previewValue=<0.5 0.5 0 1>
  }
  n2 = FunctionInput() {
   bUsePreviewValueAsDefault=true
   description="Insert UVs, world coordinates or other values to rotate."
   inputName=UVs
   inputType=FunctionInput_Vector2
   previewValue=<0.5 0.5 0 1>
  }
  n3 = Add(A=Multiply(A=@n1, B=Constant() {r=-1}), B=@n2)
  n4 = FunctionInput() {
   bUsePreviewValueAsDefault=true
   description=Enter 0-1 values for rotation. A value of 1 is equal to a 360 degree turn.
   inputName="Rotation Angle (0-1)"
   inputType=FunctionInput_Scalar
   sortPriority=2
  }
  n5 = Cosine(None=@n4) {desc=Set period to 6.283178 to match standard rotator.}
  n6 = Sine(None=@n4) {desc=Set period to 6.283178 to match standard rotator.}
  n7 = Constant() {r=-1}
  n8 = DotProduct(A=@n3, B=AppendVector(A=@n5, B=Multiply(A=@n6, B=@n7)))
  n9 = DotProduct(A=@n3, B=AppendVector(A=@n6, B=@n5))
  FunctionOutput(None=Add(A=@n1, B=AppendVector(A=@n8, B=@n9))) {outputName=Rotated Values}
 ]
}
