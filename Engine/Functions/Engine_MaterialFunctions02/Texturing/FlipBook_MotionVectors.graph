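# Material function that animates a flipbook texture, using motion vectors to smooth transitions between frames.
# A 0-1 phase and grid size select current and next frame UVs, which bias-scaled motion vectors warp forward and backward by blend phase.
# The two warped samples crossfade into color and alpha.
# Static switches allow custom UVs or costlier double-applied vectors via dependent reads, and a debug input isolates either warp direction.
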MaterialFunction /Engine/Functions/Engine_MaterialFunctions02/Texturing/FlipBook_MotionVectors
graph {
 properties {bExposeToLibrary=true, description="Flipbook with additional options for Motion Vectors. Animates a texture in a flipbook like fashion. "}
 nodes [
  n1 = FunctionInput(Preview=TextureCoordinate()) {
   bUsePreviewValueAsDefault=true
   description=Insert the texture coordinate of your choice.
   inputName=UVs
   inputType=FunctionInput_Vector2
   sortPriority=4
  }
  n2 = FunctionInput() {
   bUsePreviewValueAsDefault=true
   description=Enter the number of images horizontally across the texture.
   inputName=Number of Columns X
   inputType=FunctionInput_Scalar
   sortPriority=1
   previewValue=<2 0 0 1>
  }
  n3 = FunctionInput() {
   bUsePreviewValueAsDefault=true
   description=Enter the number of images vertically across the texture.
   inputName=Number of Rows Y
   inputType=FunctionInput_Scalar
   sortPriority=2
   previewValue=<2 0 0 1>
  }
  n4 = AppendVector(A=@n2, B=@n3)
  n5 = Divide(A=@n1, B=@n4)
  n6 = FunctionInput(Preview=Time()) {
   bUsePreviewValueAsDefault=true
   description="Apply a \"frac\" material node to the input value if time is used to control the animation phase."
   inputName="Animation  Phase (0-1)"
   inputType=FunctionInput_Scalar
  }
  n7 = Multiply(A=Multiply(A=@n2, B=@n3) {desc=Total Frames}, B=Min(A=Frac(None=@n6) {desc=0-1 animation}) {constB=1}) {desc=phase * frames}
  n8 = Floor(None=@n7) {desc=Current Frame}
  n9 = Floor(None=Divide(A=@n8, B=@n3))
  n10 = Add(A=@n5, B=Divide(A=AppendVector(A=@n8, B=@n9), B=@n4))
  n11 = FunctionInput(Preview=StaticBool()) {bUsePreviewValueAsDefault=true, inputName=Use Custom UVs, inputType=FunctionInput_StaticBool, sortPriority=11}
  n12 = StaticSwitch(
   True=FunctionInput() {bUsePreviewValueAsDefault=true, inputName=Custom UV Input, inputType=FunctionInput_Vector2, sortPriority=11}
   False=@n10
   Value=@n11
  )
  n13 = FunctionInput() {
   bUsePreviewValueAsDefault=true
   desc=Insert a texture object
   description=This texture should be linear and in the 0-1 color space. Internally it is bias-scaled into the -1 to 1 range.
   inputName=Motion Vector Intensity
   inputType=FunctionInput_Scalar
   sortPriority=8
   previewValue=<1 0 0 1>
  }
  n14 = Frac(None=ComponentMask(None=@n7) {r=true})
  n15 = FunctionOutput(None=@n14) {outputName=Blend Phase, sortPriority=2}
  n16 = FunctionInput(Preview=TextureObject() {samplerType=SAMPLERTYPE_LinearColor, texture=/Engine/ArtTools/RenderToTexture/Textures/127grey}) {
   bUsePreviewValueAsDefault=true
   desc=Insert a texture object
   description=This texture should be linear and in the 0-1 color space. Internally it is bias-scaled into the -1 to 1 range.
   inputName=Motion Vector Texture
   inputType=FunctionInput_Texture2D
   sortPriority=7
  }
  n17 = ComponentMask(None=TextureSample!(UVs=@n10, Tex=@n16, Bias, Apply View MipBias) {mipValueMode=TMVM_MipBias}.RGB) {g=true, r=true}
  n18 = Divide(A=ConstantBiasScale(None=@n17) {bias=-0.5, scale=2}, B=@n4) {desc=Scale by size of Flipbook Frame}
  n19 = Add(A=@n8) {desc=Offset 1 frame}
  n20 = Floor(None=Divide(A=@n19, B=@n3))
  n21 = Add(A=@n5, B=Divide(A=AppendVector(A=@n19, B=@n20), B=@n4)) {constB=0}
  n22 = TextureSample(UVs=@n21, Tex=@n16) {mipValueMode=TMVM_MipBias}
  n23 = Divide(A=ConstantBiasScale(None=ComponentMask(None=@n22.RGB) {g=true, r=true}) {bias=-0.5, scale=2}, B=@n4)
  n24 = ScalarParameter() {parameterName=TestAlpha}
  n25 = Multiply(A=LinearInterpolate(A=@n18, B=@n23, Alpha=@n24), B=@n13) {constB=2}
  n26 = Constant2Vector() {g=1, r=-1}
  n27 = Add(A=@n12, B=Multiply(A=@n15, B=Multiply(A=@n25, B=@n26)) {desc=Scale Motion by Frame Phase})
  n28 = ComponentMask(None=ConstantBiasScale(None=TextureSample(UVs=@n27, Tex=@n16) {mipValueMode=TMVM_MipBias}.RGB) {bias=-0.5, scale=2}) {g=true, r=true}
  n29 = Divide(A=@n28, B=@n4) {desc=Scale by size of Flipbook Frame}
  n30 = Constant2Vector() {g=1, r=-1}
  n31 = Multiply(A=Multiply(A=Multiply(A=@n13, B=@n29), B=@n30), B=@n15) {constB=2}
  n32 = FunctionInput(Preview=StaticBool()) {
   bUsePreviewValueAsDefault=true
   description="This option is more expensive but higher quality. It uses a double dependent texture read (a texture fed into UVs twice) in order to 'update' the motion vectors so they can maintain their velocities through frame transition more accurately."
   inputName=Double Apply Motion Vectors
   inputType=FunctionInput_StaticBool
   sortPriority=10
  }
  n33 = FunctionInput(Preview=TextureObject() {texture=/Engine/Functions/Engine_MaterialFunctions02/ExampleContent/Textures/flipbook}) {bUsePreviewValueAsDefault=true, desc=Insert a texture object, inputName=Texture, inputType=FunctionInput_Texture2D, sortPriority=3}
  n34 = TextureSample!(UVs=StaticSwitch(True=Add(A=@n12, B=@n31), False=@n27, Value=@n32), Tex=@n33, Apply View MipBias)
  n35 = StaticSwitch(
   True=FunctionInput() {bUsePreviewValueAsDefault=true, inputName=Custom UV2 Input, inputType=FunctionInput_Vector2, sortPriority=11}
   False=@n21
   Value=@n11
  )
  n36 = OneMinus(None=@n14) {desc=Reverse Phase}
  n37 = Multiply(A=LinearInterpolate(A=@n23, B=@n18, Alpha=@n24), B=@n13)
  n38 = Subtract(A=@n35, B=Multiply(A=@n36, B=Multiply(A=@n37, B=@n26)) {desc=Reverse Motion Vectors})
  n39 = TextureSample!(UVs=@n38, Tex=@n16, Bias, Apply View MipBias) {mipValueMode=TMVM_MipBias}
  n40 = Divide(A=ComponentMask(None=ConstantBiasScale(None=@n39.RGB) {bias=-0.5, scale=2}) {g=true, r=true}, B=@n4) {desc=Scale by size of Flipbook Frame}
  n41 = Multiply(A=Multiply(A=Multiply(A=@n13, B=@n40), B=@n30), B=@n36) {constB=2}
  n42 = TextureSample!(UVs=StaticSwitch(True=Subtract(A=@n35, B=@n41), False=@n38, Value=@n32), Tex=@n33, Apply View MipBias)
  n43 = FunctionInput() {
   bUsePreviewValueAsDefault=true
   description="If 0, Debug is Disable. If 1, Debug Forward Motion Vectors. If 2, debug Reverse Motion vectors."
   inputName=Phase to Debug
   inputType=FunctionInput_Scalar
   sortPriority=8
  }
  n44 = Floor(None=Clamp(None=@n43) {maxDefault=2})
  n45 = Min(A=@n44)
  n46 = LinearInterpolate(A=@n15, B=LinearInterpolate(Alpha=Subtract(A=@n44)), Alpha=@n45)
  FunctionOutput(None=LinearInterpolate(A=@n34.RGB, B=@n42.RGB, Alpha=@n46))
  FunctionOutput(None=LinearInterpolate(A=@n34.A, B=@n42.A, Alpha=@n46)) {outputName=Alpha, sortPriority=1}
  FunctionOutput(None=LinearInterpolate(A=@n17, B=@n28, Alpha=@n45)) {outputName=Forward Motion Vectors, sortPriority=3}
  FunctionOutput(None=@n10) {outputName=CustomUV Output, sortPriority=5}
  FunctionOutput(None=@n21) {outputName=CustomUV2 Output, sortPriority=6}
 ]
}
