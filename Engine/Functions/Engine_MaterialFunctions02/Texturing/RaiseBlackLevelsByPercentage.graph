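# RaiseBlackLevelsByPercentage is a library material function under Image Adjustment that lifts the dark values of a 0-1 image input.
# It inverts Image, multiplies it by the Black intensity level input, then inverts the product again.
# A level of 1 leaves the image unmodified and 0 turns it fully white, while white stays fixed as the range compresses linearly toward it.
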MaterialFunction /Engine/Functions/Engine_MaterialFunctions02/Texturing/RaiseBlackLevelsByPercentage
graph {
 properties {
  bExposeToLibrary=true
  description="Raises the black levels of an input value based on a percentage provided in the \"black intensity level\" parameter. Note: the image input values should range from 0-1.  "
  libraryCategoriesText [Misc, Image Adjustment]
 }
 nodes [
  n1 = OneMinus(None=FunctionInput() {inputName=Image})
  n2 = FunctionInput() {
   description="Enter 1 for unmodified black levels and 0 to raise the black levels to white. "
   inputName=Black intensity level
   sortPriority=1
  }
  FunctionOutput(None=OneMinus(None=Multiply(A=@n1, B=@n2) {constB=0.6}))
 ]
}
