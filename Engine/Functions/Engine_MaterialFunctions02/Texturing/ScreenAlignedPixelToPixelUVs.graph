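# ScreenAlignedPixelToPixelUVs is a library material function that produces UVs mapping a texture to the screen one texel per screen pixel, without stretching.
# It divides the screen pixel position by a TextureResolution Vector2 input, which defaults to 128 by 128.
# Its description notes that it previews incorrectly in the material editor.
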MaterialFunction /Engine/Functions/Engine_MaterialFunctions02/Texturing/ScreenAlignedPixelToPixelUVs
graph {
 properties {
  bExposeToLibrary=true
  description=Maps a texture to the screen without stretching any of the pixels. Each pixel on the screen will line up with a pixel from the sampled texture. Note that this will preview incorrectly in the material editor.
  libraryCategoriesText [Misc,texturing]
 }
 nodes [
  n1 = FunctionInput() {
   bUsePreviewValueAsDefault=true
   description="Input the x and y dimensions of the texture you're referencing. "
   inputName=TextureResolution
   inputType=FunctionInput_Vector2
   previewValue=<128 128 0 1>
  }
  FunctionOutput(None=Divide(A=ScreenPosition().PixelPosition, B=@n1))
 ]
}
