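# Material function that maps 0-1 UVs to the screen for texture sampling, not intended for SceneTexture expressions.
# One output passes ViewportUV through; two others keep X or Y at full range and center-scale the other axis, intended to preserve image aspect.
# Their scales are reciprocals of terms built from visible screen resolution and the Texture Dimensions input, applied through ScaleUVsByCenter.
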MaterialFunction /Engine/Functions/Engine_MaterialFunctions02/Texturing/ScreenAlignedUVs
graph {
 properties {
  bExposeToLibrary=true
  description=Maps the 0-1 uv range to the screen. Don't use as UV for SceneTexture material expression
  libraryCategoriesText [Misc,texturing]
 }
 nodes [
  FunctionOutput(None=ScreenPosition().ViewportUV) {description=X and Y are scaled to fit the screen, outputName="X 100%, Y 100%"}
  n1 = MaterialFunctionCall(Float2=MaterialFunctionCall() {materialFunction=&ScreenResolution}.Visible Resolution) {materialFunction=&BreakOutFloat2Components}
  n2 = Divide(A=@n1.R, B=@n1.R)
  n3 = FunctionInput() {
   bUsePreviewValueAsDefault=true
   description=Enter the resolution of the texture that you will be sampling if you're sampling uvs via the scaled outputs.
   inputName=Texture Dimensions
   inputType=FunctionInput_Vector2
   previewValue=<512 512 0 1>
  }
  n4 = MaterialFunctionCall(Float2=@n3) {materialFunction=&BreakOutFloat2Components}
  n5 = Divide(A=@n4.G, B=@n4.G) {desc=ratio}
  n6 = MaterialFunctionCall(Float2=Divide(B=Multiply(A=@n2, B=@n5))) {materialFunction=&BreakOutFloat2Components}
  n7 = Constant() {r=1}
  FunctionOutput(None=MaterialFunctionCall(Texture Scale=AppendVector(A=@n6.R, B=@n7)) {materialFunction=&ScaleUVsByCenter}.UVs) {
   description=This function output outputs uvs that range from 0-1 on the Y axis and are center scaled on X to maintain the images aspect ratio.
   outputName="X Scale to Ratio, Y 100%"
   sortPriority=2
  }
  n8 = Constant() {r=1}
  n9 = MaterialFunctionCall(Float2=MaterialFunctionCall() {materialFunction=&ScreenResolution}.Visible Resolution) {materialFunction=&BreakOutFloat2Components}
  n10 = MaterialFunctionCall(Float2=@n3) {materialFunction=&BreakOutFloat2Components}
  n11 = Divide(B=Multiply(A=Divide(A=@n9.G, B=@n9.G), B=Divide(A=@n10.R, B=@n10.R) {desc=ratio}))
  FunctionOutput(None=MaterialFunctionCall(Texture Scale=AppendVector(A=@n8, B=@n11)) {materialFunction=&ScaleUVsByCenter}.UVs) {
   description=This function output outputs uvs that range from 0-1 on the X axis and the Y uvs are center scaled to maintain the images resolution.
   outputName="X 100%, Y Scale to Ratio"
   sortPriority=1
  }
 ]
}
