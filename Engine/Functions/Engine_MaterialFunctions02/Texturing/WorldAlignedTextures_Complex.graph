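# Material function that tiles diffuse, spec and normal textures in world space by triplanar projection, with optional separate Z-plane settings.
# X, Y and Z samples blend by contrast-adjusted world normal, optionally refined by height lerps using diffuse alpha or a chosen channel.
# WorldAlignedNormals_HighQuality reuses those blend weights for world or tangent space normals.
# Outputs cover XYZ, XY and Z results as textures and material attributes, plus a flat-top normal.
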MaterialFunction /Engine/Functions/Engine_MaterialFunctions02/Texturing/WorldAlignedTextures_Complex
graph {
 properties {bExposeToLibrary=true, description=Tiles a texture in worldspace, libraryCategoriesText [Misc,Texturing]}
 nodes [
  FunctionInput() {
   bUsePreviewValueAsDefault=true
   description=-------------
   inputName=--------------------------
   inputType=FunctionInput_StaticBool
   sortPriority=-10
  }
  FunctionInput() {
   bUsePreviewValueAsDefault=true
   description=-------------
   inputName="--------------------------     "
   inputType=FunctionInput_StaticBool
   sortPriority=199
  }
  FunctionInput() {
   bUsePreviewValueAsDefault=true
   description=-------------
   inputName="--------------------------  "
   inputType=FunctionInput_StaticBool
   sortPriority=104
  }
  FunctionInput() {
   bUsePreviewValueAsDefault=true
   description=-------------
   inputName="--------------------------       "
   inputType=FunctionInput_StaticBool
   sortPriority=10
  }
  FunctionInput() {
   bUsePreviewValueAsDefault=true
   description=-------------
   inputName="--------------------------      "
   inputType=FunctionInput_StaticBool
   sortPriority=5
  }
  n1 = FunctionInput(Preview=WorldPosition().XYZ) {bUsePreviewValueAsDefault=true, inputName=WorldPosition, sortPriority=200}
  n2 = FunctionInput() {
   bUsePreviewValueAsDefault=true
   description=Size of texture in worldspace units
   inputName=TextureSize
   sortPriority=-11
   previewValue=<64 64 64 6>
  }
  n3 = Multiply(A=Abs(None=@n2)) {constB=-1}
  n4 = Divide(A=@n1, B=@n3)
  n5 = ComponentMask(None=@n4) {b=true, r=true}
  n6 = TextureObject() {texture=/Engine/EngineMaterials/DefaultDiffuse}
  n7 = FunctionInput(Preview=@n6) {bUsePreviewValueAsDefault=true, inputName=Diffuse Texture Object, inputType=FunctionInput_Texture2D, sortPriority=-100}
  n8 = TextureSample(UVs=@n5, Tex=@n7) {desc=Y projection, samplerSource=SSM_Wrap_WorldGroupSettings}
  n9 = AppendVector(A=@n8.RGB, B=@n8.RGB)
  n10 = FunctionInput(Preview=StaticBool()) {
   bUsePreviewValueAsDefault=true
   description="Default = False"
   inputName=Export Float 4 Diffuse
   inputType=FunctionInput_StaticBool
   sortPriority=105
  }
  n11 = ComponentMask(None=@n4) {b=true, g=true}
  n12 = TextureSample(UVs=@n11, Tex=@n7) {desc=X projection, samplerSource=SSM_Wrap_WorldGroupSettings}
  n13 = AppendVector(A=@n12.RGB, B=@n12.RGB)
  n14 = FunctionInput(Preview=VertexNormalWS()) {bUsePreviewValueAsDefault=true, inputName=World Space Normal, sortPriority=201, previewValue=<0 0 1 1>}
  n15 = ComponentMask(None=@n14) {r=true}
  n16 = FunctionInput() {bUsePreviewValueAsDefault=true, inputName=ProjectionTransitionContrast, inputType=FunctionInput_Scalar, sortPriority=7}
  n17 = ComponentMask(None=MaterialFunctionCall(In=Abs(None=@n15), Contrast=@n16) {materialFunction=&CheapContrast}.Result) {desc="1=+x", r=true}
  n18 = FunctionInput() {
   bUsePreviewValueAsDefault=true
   description="(If pick the diffuse height channel == true)  (1,0,0,0) will use the diffuse red channel as the height map. (0,1,0,0) will use the green channel etc."
   inputName=TheDiffuseHeightChannel
   inputType=FunctionInput_Vector4
   sortPriority=9
  }
  n19 = FunctionInput(Preview=StaticBool()) {
   bUsePreviewValueAsDefault=true
   description="Default = False"
   inputName=PickTheDiffuseHeightChannel
   inputType=FunctionInput_StaticBool
   sortPriority=9
  }
  n20 = FunctionInput() {bUsePreviewValueAsDefault=true, inputName=HeightLerpTransitionContrast, inputType=FunctionInput_Scalar, sortPriority=9}
  n21 = MaterialFunctionCall!(
   Diffuse 1
   Diffuse 2
   Transistion Phase=@n17
   Height Texture 1=StaticSwitch(True=DotProduct(A=@n9, B=@n18), False=@n8.A, Value=@n19)
   Height Texture 2=StaticSwitch(True=DotProduct(A=@n13, B=@n18), False=@n12.A, Value=@n19)
   Contrast=@n20
  ) {materialFunction=&/HeightLerpWithTwoHeightMaps}
  n22 = OneMinus(None=@n21.Alpha)
  n23 = FunctionInput(Preview=StaticBool()) {
   bUsePreviewValueAsDefault=true
   description=Defaults to false.
   inputName=Use Texture Alpha Heightmap to Refine the Transistion
   inputType=FunctionInput_StaticBool
   sortPriority=8
   previewValue=<1 0 0 1>
  }
  n24 = StaticSwitch(True=@n22, False=@n17, Value=@n23)
  n25 = LinearInterpolate(A=StaticSwitch(True=@n9, False=@n8.RGB, Value=@n10), B=StaticSwitch(True=@n13, False=@n12.RGB, Value=@n10), Alpha=@n24) {desc="Float3 1 = x projection"}
  n26 = FunctionInput() {
   bUsePreviewValueAsDefault=true
   description=Size of texture in worldspace units
   inputName=Texture Size Z projection
   previewValue=<64 64 64 6>
  }
  n27 = Multiply(A=@n26) {constB=-1}
  n28 = FunctionInput(Preview=StaticBool()) {
   bUsePreviewValueAsDefault=true
   inputName=Use seperate settings for Z projection
   inputType=FunctionInput_StaticBool
   previewValue=<64 64 64 6>
  }
  n29 = ComponentMask(None=StaticSwitch(True=Divide(A=@n1, B=@n27), False=@n4, Value=@n28)) {g=true, r=true}
  n30 = FunctionInput(Preview=@n6) {bUsePreviewValueAsDefault=true, inputName=Z Plane Diffuse Texture Object, inputType=FunctionInput_Texture2D, sortPriority=4}
  n31 = TextureSample(UVs=@n29, Tex=@n30) {samplerSource=SSM_Wrap_WorldGroupSettings}
  n32 = TextureSample(UVs=@n29, Tex=@n7) {samplerSource=SSM_Wrap_WorldGroupSettings}
  n33 = FunctionInput(Preview=StaticBool()) {
   bUsePreviewValueAsDefault=true
   inputName=Use seperate texture for Z projection
   inputType=FunctionInput_StaticBool
   previewValue=<64 64 64 6>
  }
  n34 = StaticSwitch(True=AppendVector(A=@n31.RGB, B=@n31.RGB), False=AppendVector(A=@n32.RGB, B=@n32.RGB), Value=@n33) {desc=Z projection}
  n35 = StaticSwitch(True=@n34, False=ComponentMask(None=@n34) {b=true, g=true, r=true}, Value=@n10)
  n36 = ComponentMask(None=@n14) {b=true}
  n37 = ComponentMask(None=MaterialFunctionCall(In=Abs(None=@n36), Contrast=@n16) {materialFunction=&CheapContrast}.Result) {desc="1=+z", r=true}
  n38 = DotProduct(A=@n34, B=@n18)
  n39 = ComponentMask(None=@n34) {a=true}
  n40 = OneMinus(None=MaterialFunctionCall(
    Transistion Phase=@n37
    Height Texture 1=@n21.Resulting Height Map
    Height Texture 2=StaticSwitch(True=@n38, False=@n39, Value=@n19)
    Contrast=@n20
   ) {materialFunction=&HeightLerpWithTwoHeightMaps}.Alpha)
  n41 = StaticSwitch(True=@n40, False=@n37, Value=@n23)
  n42 = FunctionOutput(None=LinearInterpolate(A=@n25, B=@n35, Alpha=@n41) {desc="1 = z projection"}) {outputName=XYZ Diffuse Texture, sortPriority=1}
  n43 = TextureObject() {texture=&DefaultDiffuse}
  n44 = FunctionInput(Preview=@n43) {
   bUsePreviewValueAsDefault=true
   description="The spec texture is unpacked as Red channel = specular, Green channel = Roughness, Blue Channel = Metallic when referencing the Material Attribute outputs."
   inputName=Spec Texture Object
   inputType=FunctionInput_Texture2D
   sortPriority=-100
  }
  n45 = TextureSample(UVs=@n5, Tex=@n44) {desc=Y projection, samplerSource=SSM_Wrap_WorldGroupSettings}
  n46 = FunctionInput(Preview=StaticBool()) {
   bUsePreviewValueAsDefault=true
   description="Default = False"
   inputName=Export Float 4 Spec
   inputType=FunctionInput_StaticBool
   sortPriority=106
  }
  n47 = TextureSample(UVs=@n11, Tex=@n44) {desc=X projection, samplerSource=SSM_Wrap_WorldGroupSettings}
  n48 = LinearInterpolate(
   A=StaticSwitch(True=AppendVector(A=@n45.RGB, B=@n45.RGB), False=@n45.RGB, Value=@n46)
   B=StaticSwitch(True=AppendVector(A=@n47.RGB, B=@n47.RGB), False=@n47.RGB, Value=@n46)
   Alpha=@n24
  ) {desc="Float3 1 = x projection"}
  n49 = FunctionInput(Preview=@n43) {
   bUsePreviewValueAsDefault=true
   description="The spec texture is unpacked as Red channel = specular, Green channel = Roughness, Blue Channel = Metallic when referencing the Material Attribute outputs."
   inputName=Z Plane Spec Texture Object
   inputType=FunctionInput_Texture2D
   sortPriority=4
  }
  n50 = TextureSample(UVs=@n29, Tex=@n49) {samplerSource=SSM_Wrap_WorldGroupSettings}
  n51 = TextureSample(UVs=@n29, Tex=@n44) {samplerSource=SSM_Wrap_WorldGroupSettings}
  n52 = StaticSwitch(True=AppendVector(A=@n50.RGB, B=@n50.RGB), False=AppendVector(A=@n51.RGB, B=@n51.RGB), Value=@n28) {desc=Z projection}
  n53 = StaticSwitch(True=@n52, False=ComponentMask(None=@n52) {b=true, g=true, r=true}, Value=@n46)
  n54 = MaterialFunctionCall(Float3=FunctionOutput(None=LinearInterpolate(A=@n48, B=@n53, Alpha=@n41) {desc="1 = z projection"}) {outputName=XYZ Spec, sortPriority=11}) {materialFunction=&BreakOutFloat3Components}
  n55 = ComponentMask(None=@n42) {a=true}
  n56 = Constant()
  n57 = MaterialFunctionCall!(
   Negative World Position Divisor=@n3
   World Position=@n1
   Texture=FunctionInput(Preview=TextureObject() {samplerType=SAMPLERTYPE_LinearColor, texture=/Engine/EngineMaterials/DefaultNormal_Uncompressed}) {bUsePreviewValueAsDefault=true, inputName=NormalTextureObject, inputType=FunctionInput_Texture2D, sortPriority=30}
   Z lerp Alpha=StaticSwitch(True=@n40, False=@n37, Value=@n23)
   WorldNormal=@n14
   X lerp alpha=StaticSwitch(True=@n22, False=@n17, Value=@n23)
  ) {materialFunction=/Engine/Functions/Engine_MaterialFunctions02/WorldAlignedNormals_HighQuality}
  n58 = FunctionInput(Preview=StaticBool() {value=true}) {
   bUsePreviewValueAsDefault=true
   description="This switch defaults to true. If left true ensure that tangent space normals is turned off in the shaders properties panel. "
   inputName=World Space Normals
   inputType=FunctionInput_StaticBool
   sortPriority=31
  }
  n59 = StaticSwitch(
   True=@n57.World Space Normals XYZ
   False=Transform(None=@n57.World Space Normals XYZ) {transformSourceType=TRANSFORMSOURCE_World, transformType=TRANSFORM_Tangent}
   Value=@n58
  )
  FunctionOutput(None=MakeMaterialAttributes!(
    BaseColor=@n42
    Metallic=@n54.B
    Specular=@n54.B
    Roughness=@n54.B
    Anisotropy
    EmissiveColor
    Opacity=StaticSwitch(True=@n55, False=@n56, Value=@n10)
    OpacityMask
    Normal=FunctionOutput(None=@n59) {outputName=Normal XYZ Texture, sortPriority=30}
    Tangent
    WorldPositionOffset
    SubsurfaceColor
    ClearCoat
    ClearCoatRoughness
    AmbientOcclusion
    Refraction
    CustomizedUVs_0
    CustomizedUVs_1
    CustomizedUVs_2
    CustomizedUVs_3
    CustomizedUVs_4
    CustomizedUVs_5
    CustomizedUVs_6
    CustomizedUVs_7
    PixelDepthOffset
    ShadingModel
    Displacement
   )) {
   description="The spec texture is unpacked as Red channel = specular, Green channel = Roughness, Blue Channel = Metallic. Diffuse alpha is in the opacity output if export float 4 is set to true."
   outputName=XYZ Mat Attributes - See Tooltip
   sortPriority=51
  }
  FunctionOutput() {outputName=" --------------------------", sortPriority=20}
  FunctionOutput() {outputName="  --------------------------", sortPriority=45}
  FunctionOutput() {outputName="-------------------------- ", sortPriority=4}
  FunctionOutput(None=StaticSwitch(
    True=@n57.XY Flat Top Projection World Space Normals
    False=Transform(None=@n57.XY Flat Top Projection World Space Normals) {transformSourceType=TRANSFORMSOURCE_World, transformType=TRANSFORM_Tangent}
    Value=@n58
   )) {outputName=Normal XYZ Flat Top, sortPriority=31}
  n60 = FunctionOutput(None=@n25) {outputName=XY Diffuse Texture, sortPriority=2}
  n61 = MaterialFunctionCall(Float3=FunctionOutput(None=@n48) {outputName=XY Spec, sortPriority=12}) {materialFunction=&BreakOutFloat3Components}
  n62 = ComponentMask(None=@n60) {a=true}
  n63 = StaticSwitch(
   True=@n57.XY Projection World Space Normals
   False=Transform(None=@n57.XY Projection World Space Normals) {transformSourceType=TRANSFORMSOURCE_World, transformType=TRANSFORM_Tangent}
   Value=@n58
  )
  FunctionOutput(None=MakeMaterialAttributes(
    BaseColor=@n60
    Metallic=@n61.B
    Specular=@n61.B
    Roughness=@n61.B
    Opacity=StaticSwitch(True=@n62, False=@n56, Value=@n10)
    Normal=FunctionOutput(None=@n63) {outputName=Normal XY Texture, sortPriority=32}
   )) {
   description="The spec texture is unpacked as Red channel = specular, Green channel = Roughness, Blue Channel = Metallic. Diffuse alpha is in the opacity output if export float 4 is set to true."
   outputName=XY Mat Attributes - See Tooltip
   sortPriority=52
  }
  n64 = FunctionOutput(None=@n35) {outputName=Z Diffuse Texture, sortPriority=3}
  n65 = MaterialFunctionCall(Float3=FunctionOutput(None=@n53) {outputName=Z Spec, sortPriority=13}) {materialFunction=&BreakOutFloat3Components}
  n66 = ComponentMask(None=@n64) {a=true}
  n67 = StaticSwitch(
   True=@n57.Z Projection World Space Normals
   False=Transform(None=@n57.Z Projection World Space Normals) {transformSourceType=TRANSFORMSOURCE_World, transformType=TRANSFORM_Tangent}
   Value=@n58
  )
  FunctionOutput(None=MakeMaterialAttributes(
    BaseColor=@n64
    Metallic=@n65.B
    Specular=@n65.B
    Roughness=@n65.B
    Opacity=StaticSwitch(True=@n66, False=@n56, Value=@n10)
    Normal=FunctionOutput(None=@n67) {outputName=Normal Z Texture, sortPriority=33}
   )) {
   description="The spec texture is unpacked as Red channel = specular, Green channel = Roughness, Blue Channel = Metallic. Diffuse alpha is in the opacity output if export float 4 is set to true."
   outputName=Z Mat Attributes - See Tooltip
   sortPriority=53
  }
 ]
}
